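; Player controller component that displays one HUD notification at a time on the owning client, callable locally or from the server.
; A new notification replaces the current one immediately; interrupted infinite ones queue at the front to return later, while timed ones are dropped.
; Timers remove timed notifications, and removal promotes the next queued one.
; Listeners receive data through a dispatcher, and a delta-gated toggle adds or removes the DynamicUI scene.
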(event EventBeginPlay
  (|CallParentFunction)
  (bind _asplayer_controller (Utilities|Casting|CastToPlayerController (Components|GetOwner)))
  (Variables|Default|SetOwningPlayerController _asplayer_controller))

(event Custom|DisplayNotification_OnOwningClient
  (params (NotificationData :type BP%20Notification%20Data%20Type%20Object%20Reference))
  (Public|DisplayNotification NotificationData))

(event Custom|RemoveNotification_OnOwningClient
  (params (NotificationData :type BP%20Notification%20Data%20Type%20Object%20Reference))
  (Public|RemoveNotification NotificationData))

(fn AddNotificationData
  (params (NotificationData :type BP%20Notification%20Data%20Type%20Object%20Reference))
  (if (!= NotificationData (Variables|Default|GetCurrentNotificationData))
    (Utilities|IsValid (Variables|Default|GetCurrentNotificationData)
      (:Is%20Valid
        ; Handle active notification
        (if (not (> (Variables|/Game/Athena/HUD/Messaging/BP_NotificationDataType.BP_NotificationDataType_C|GetDuration (Variables|Default|GetCurrentNotificationData)) 0))
          (Utilities|Array|Insert (Variables|Default|GetQueuedNotifications) (Variables|Default|GetCurrentNotificationData)) ; Otherwise, add this to the front of\nthe queue to return later
        ))
      (:Is%20Not%20Valid
        _))
    (bind _output_get (Variables|Default|SetCurrentNotificationData (Variables|GetNotificationData)))
    (Default|CallOnNewNotificationData _output_get)
    (if (> (Variables|/Game/Athena/HUD/Messaging/BP_NotificationDataType.BP_NotificationDataType_C|GetDuration _output_get) 0)
      ; Set timer for current notification to be removed
      (bind _returnvalue (Utilities|Time|SetTimerbyEvent (EventDispatchers|CreateEvent :!event RemoveCurrentNotificationData) (Variables|/Game/Athena/HUD/Messaging/BP_NotificationDataType.BP_NotificationDataType_C|GetDuration _output_get)))
      (Variables|Default|SetOnRemoveNotificationTimer _returnvalue)
      (else
        ; If this notification is infinite, clear any existing timer
        (Utilities|Time|ClearandInvalidateTimerbyHandle (Variables|Default|GetOnRemoveNotificationTimer))))))

(fn DisplayNotification
  (params (NotificationData :type BP%20Notification%20Data%20Type%20Object%20Reference))
  (Networking|CanExecuteCosmeticEvents
    (:True
      (Internal|AddNotificationData NotificationData)
      (Internal|SetNotificationSceneActive true))))

(fn DisplayNotification_FromServer
  (params (NotificationData :type BP%20Notification%20Data%20Type%20Object%20Reference))
  (Networking|SwitchHasAuthority
    (:Authority
      (CallFunction|DisplayNotificationonOwningClient NotificationData))))

(fn GetNotificationCurrentlyDisplaying
  (return (Variables|Default|GetCurrentNotificationData)))

(fn RemoveCurrentNotificationData
  (Public|RemoveNotification (Variables|Default|GetCurrentNotificationData)))

(fn RemoveNotification
  (params (NotificationData :type BP%20Notification%20Data%20Type%20Object%20Reference))
  (bind _queuednotifications (Variables|Default|GetQueuedNotifications))
  (Networking|CanExecuteCosmeticEvents
    (:True
      (if (== NotificationData (Variables|Default|GetCurrentNotificationData))
        (Utilities|Time|ClearandInvalidateTimerbyHandle (Variables|Default|GetOnRemoveNotificationTimer))
        (if (> (Utilities|Array|Length (Variables|Default|GetQueuedNotifications)) 0)
          ; If anything is in the queue, set the current data and notify listeners.  Currently only infinite data gets added  to the queue, so no extra timer handling needed
          (Variables|Default|SetCurrentNotificationData (Utilities|Array|Get(acopy) _queuednotifications))
          (Utilities|Array|RemoveIndex _queuednotifications)
          (Default|CallOnNewNotificationData (Variables|Default|GetCurrentNotificationData))
          (else
            ; If nothing is left in the queue, remove the scene
            (Internal|SetNotificationSceneActive false)
            (Variables|Default|SetCurrentNotificationData 0)))
        (else
          ; If this is not active but in the queue, it is safe to just remove
          (Utilities|Array|RemoveItem (Variables|Default|GetQueuedNotifications) (Variables|GetNotificationData)))))))

(fn SetNotificationSceneActive
  (params (IsActive :type Boolean))
  (bind _returnvalue (WorldSubsystems|GetDynamicUIManager))
  (bind _owningplayercontroller (Variables|Default|GetOwningPlayerController))
  (bind _notificationscene (Variables|Default|GetNotificationScene))
  (bind _updated_variable (Utilities|FlowControl|DeltaGateWildcard (Variables|Default|GetIsNotificationActive) IsActive)
    (:Success
      (Utilities|IsValid (Variables|Default|GetOwningPlayerController)
        (:Is%20Valid
          (if _updated_variable
            (DynamicUI|AddScene _returnvalue _notificationscene _owningplayercontroller)
            (else
              (DynamicUI|RemoveScene _returnvalue _notificationscene _owningplayercontroller))))))))
