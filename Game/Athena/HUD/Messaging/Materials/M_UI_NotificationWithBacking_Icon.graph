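# Translucent UI material that, going by its name, draws the foreground icon of a HUD notification with a backing.
# A ForegroundIcon texture parameter, defaulting to an empty texture, is tinted by the ForegroundIconColor vector parameter.
# RGB and alpha are multiplied separately and saturated, so the color parameter controls both the icon's tint and its transparency.
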Material /Game/Athena/HUD/Messaging/Materials/M_UI_NotificationWithBacking_Icon
graph {
 properties {blendMode=BLEND_Translucent, materialDomain=MD_UI, pixelDepthOffsetMode=PDOM_Legacy, usedShadingModels=Default Lit}
 nodes [
  n1 = VectorParameter() {parameterName=ForegroundIconColor, defaultValue=<rgba 1 0 0 1>}
  n2 = TextureSampleParameter2D() {parameterName=ForegroundIcon, texture=/Game/Athena/UI/Frontend/Spatial/Art/T_Ui_Empty}
  Saturate(None=Multiply(A=@n1.A, B=@n2.A))
  Saturate(None=Multiply(A=@n1.RGB, B=@n2.RGB))
 ]
}
