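# UI material that draws a notification backplate with rarity-colored brackets and glow over a dark translucent fill, using alpha-composite blending.
# Mirrored, aspect-corrected UVs drive angled-box masks that form the brackets and inner box procedurally, with parameters for offset and stretch animation.
# A trapezoid SDF and reflected gradient form the glow, tinted separately from the brackets, which can blink to white.
# Opacity parameters control fill, glow and overall visibility.
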Material /Game/Athena/UI/Materials/M_UI_NotificationBackplate
graph {
 properties {
  bAutomaticallySetUsageInEditor=false
  bUsedWithStaticMesh=false
  blendMode=BLEND_AlphaComposite
  materialDomain=MD_UI
  usedShadingModels=Default Lit
 }
 nodes [
  n1 = MaterialFunctionCall() {materialFunction=&GetUserInterfaceUV}
  n2 = MaterialFunctionCall!(
   UV=NamedRerouteDeclaration(None=MaterialFunctionCall!(UVs=@n1.Normalized UV) {materialFunction=/FortUILibrary/MaterialFunctions/Basics/MF_UI_9SliceUVs}.UVs) {name=UV, nodeColor=<rgba 0 0 0 1>}
   Mirror Y
   Swizzle UV
  ) {materialFunction=/FortUILibrary/MaterialFunctions/Basics/MF_UI_Symmetry_2x1_New}
  n3 = Constant() {r=0.5}
  n4 = ScalarParameter() {parameterName=XOffset}
  n5 = Constant()
  n6 = MaterialFunctionCall(Float2=@n1.Normalized UV) {materialFunction=&BreakOutFloat2Components}
  LinearInterpolate(
   A=Add(
    A=MaterialFunctionCall(UVs=@n2.Symmetry, Texture Scale=@n3) {materialFunction=&ScaleUVsByCenter}.UVs
    B=AppendVector(A=@n4, B=@n5)
   )
   B=AppendVector(A=OneMinus(None=@n6.R), B=@n6.G)
   Alpha=ScalarParameter() {parameterName=FlipX}
  )
  n7 = NamedRerouteUsage()
  n8 = ScalarParameter() {defaultValue=0.4, parameterName=BackplateOpacity}
  n9 = ScalarParameter() {defaultValue=1, parameterName=Opacity}
  Multiply(A=Max(A=Multiply(A=@n7, B=@n8) {desc=BackplateOpacity}, B=NamedRerouteUsage()), B=@n9) {constB=0}
  n10 = VectorParameter() {parameterName=Rarity03, defaultValue=<rgba 0.0044 0.0529 0.9216 1>}
  n11 = NamedRerouteUsage()
  n12 = VectorParameter() {parameterName=Rarity01, defaultValue=<rgba 0.0331 1 0.9131 1>}
  n13 = Constant3Vector() {constant=<rgba 1 1 1 0>}
  n14 = ScalarParameter() {parameterName=BracketsBlink}
  n15 = Max(
   A=Multiply(A=Multiply(A=@n10.RGB, B=@n11), B=ScalarParameter() {defaultValue=0.5, parameterName=GlowOpacity}) {constB=0.5, desc=Glow Opacity}
   B=Multiply(A=LinearInterpolate(A=@n12.RGB, B=@n13, Alpha=@n14), B=NamedRerouteUsage())
  )
  Multiply(A=VertexColor().A, B=Multiply(A=LinearInterpolate(A=@n15, Alpha=@n7) {constB=0}, B=@n9) {constB=0})
  n16 = MaterialFunctionCall() {materialFunction=&GetUserInterfaceUV}
  n17 = MaterialFunctionCall(UV=@n16.Normalized UV) {materialFunction=&MF_UI_Symmetry_2x1_New}
  n18 = ComponentMask(None=@n17.Symmetry) {r=true}
  n19 = Constant() {desc=OffsetX, r=-0.05}
  n20 = ScalarParameter() {parameterName=AnimOffset}
  n21 = Multiply(
   A=Divide(A=ComponentMask(None=@n16.Pixel Size) {r=true}, B=ComponentMask(None=@n16.Pixel Size) {g=true})
   B=Subtract(A=Subtract(A=@n18, B=@n19), B=@n20) {constB=0.5}
  )
  n22 = ComponentMask(None=@n17.Symmetry) {g=true}
  n23 = Constant() {desc=OffsetY}
  n24 = MaterialFunctionCall(UVs=AppendVector(A=Add(A=@n21, B=@n20) {constB=0.5}, B=Subtract(A=@n22, B=@n23)), Texture Scale=Constant() {r=0.6}) {materialFunction=&ScaleUVsByCenter}
  n25 = Constant()
  n26 = ScalarParameter() {defaultValue=8.6787, parameterName=TopAngle}
  n27 = LinearInterpolate(
   A=@n26
   B=Subtract(A=@n26, B=ScalarParameter() {defaultValue=1.3816, parameterName=StretchAmount_Top}) {constB=3}
   Alpha=ScalarParameter() {defaultValue=1.4212, parameterName=Anim_Stretch}
  )
  n28 = MaterialFunctionCall(Float2=@n24.UVs) {materialFunction=&BreakOutFloat2Components}
  n29 = AppendVector(A=@n28.R, B=OneMinus(None=@n28.G))
  n30 = Constant()
  n31 = ScalarParameter() {defaultValue=7.777, parameterName=BottomAngle}
  n32 = LinearInterpolate(
   A=@n31
   B=Subtract(A=@n31, B=ScalarParameter() {defaultValue=1.2457, parameterName=StretchAmount_Bottom}) {constB=3}
   Alpha=ScalarParameter() {defaultValue=1.4212, parameterName=Anim_Stretch}
  )
  n33 = Min(
   A=Saturate(None=MaterialFunctionCall!(UV=@n24.UVs, Corner=@n25, Angle=@n27, Flip, Falloff) {materialFunction=/FortUILibrary/MaterialFunctions/Shapes/MF_UI_BoxAngle}.Result)
   B=Saturate(None=MaterialFunctionCall(UV=@n29, Corner=@n30, Angle=@n32) {materialFunction=&MF_UI_BoxAngle}.Result)
  )
  n34 = Saturate(None=MaterialFunctionCall(UV=@n29, Corner=Constant() {r=3}, Angle=ScalarParameter() {defaultValue=33.9908, parameterName=OustideAngle}) {materialFunction=&MF_UI_BoxAngle}.Result)
  n35 = Add(A=@n29, B=ScalarParameter() {defaultValue=-0.126, parameterName=InnerAngleOffset})
  n36 = Constant() {r=3}
  n37 = ScalarParameter() {defaultValue=29.9303, parameterName=InnerAngle}
  n38 = MaterialFunctionCall(
   DistanceField=ComponentMask(None=Sine(None=@n24.UVs) {period=2}) {g=true}
   Threshold=ScalarParameter() {defaultValue=0.3303, parameterName=CenterCutHeight}
  ) {materialFunction=&MF_UI_AntiAliasing}
  n39 = Min(
   A=Saturate(None=MaterialFunctionCall(UV=@n35, Corner=@n36, Angle=@n37) {materialFunction=&MF_UI_BoxAngle}.Result)
   B=Ceil(None=Saturate(None=@n38.Outer))
  )
  NamedRerouteDeclaration(None=Min(A=Min(A=@n33, B=@n34), B=OneMinus(None=@n39))) {name=Shape, nodeColor=<rgba 0 0 0 1>}
  NamedRerouteDeclaration(None=@n39) {name=InnerBox, nodeColor=<rgba 1 0 0.4453 1>}
  n40 = Add(A=MaterialFunctionCall!(
    UVs
    TopWidth=Constant() {r=0.5}
    BottomWidth=Constant() {r=0.4}
    Height=Constant() {r=0.1}
    OutlineThickness
    OutlineAlignment
    GlowMin
    GlowMax
   ) {materialFunction=/FortUILibrary/MaterialFunctions/Shapes/MF_UI_IsoscelesTrapezoid}.SDF) {constB=0.5}
  n41 = Power(Base=MaterialFunctionCall!(UVs, Offset, RotationCenter, RotationAngle) {materialFunction=/FortUILibrary/MaterialFunctions/Gradients/MF_UI_ReflectedGradient}.Result)
  NamedRerouteDeclaration(None=Multiply(A=OneMinus(None=@n40), B=Multiply(A=@n41))) {name=Glow, nodeColor=<rgba 0 0 0 1>}
  NamedRerouteUsage()
 ]
}
