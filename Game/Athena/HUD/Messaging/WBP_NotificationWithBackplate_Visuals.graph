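# Visual widget for an Athena HUD notification: an icon, uppercase headline and optional rich-text detail on an animated material backplate.
# Setter functions fill text, backplate material, colors and both icon images; compact touch UI narrows text wrapping.
# The intro drives backplate material parameters and pops an enlarged duplicate icon before revealing the content, alongside an optional sound.
# The outro fades the whole widget, and named slots allow custom icon and header content.
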WidgetBlueprint /Game/Athena/HUD/Messaging/WBP_NotificationWithBackplate_Visuals
animations [
 Anim_Intro {
  binding_count=4
  length_seconds=1.75
  object=&self:Anim_Intro
  sequence {bindings [
    PrimaryWidthSB {
     id=#1
     possessed_class=SizeBox
     tracks [{
       type=MovieSceneFloatTrack
       display_name=Render Opacity
       sections [{type=MovieSceneFloatSection, end_frame=35, start_frame=0, channels [None {type=MovieSceneScriptingFloatChannel, const [1,2]}]}]
      }]
    }
    HorizontalBox_1 {
     id=#2
     possessed_class=HorizontalBox
     tracks [{
       type=MovieSceneFloatTrack
       display_name=Render Opacity
       sections [{type=MovieSceneFloatSection, end_frame=35, start_frame=0, channels [None {type=MovieSceneScriptingFloatChannel, keys [[0,0],[32,0],[35,1]]}]}]
      }]
    }
    AnimBoundCB {
     id=#3
     possessed_class=CommonBorder
     tracks [{
       type=MovieSceneWidgetMaterialTrack
       sections [{
         type=MovieSceneParameterSection
         end_frame=35
         start_frame=0
         channels [
          Opacity {type=MovieSceneScriptingFloatChannel, keys [[0,0],[5,1]]}
          GlowOpacity {type=MovieSceneScriptingFloatChannel, keys [[0,0],[30,0],[35,1]]}
          AnimOffset {type=MovieSceneScriptingFloatChannel, keys [[0,0.5],[5,1],[26,1],[35,0]]}
          BracketsBlink {type=MovieSceneScriptingFloatChannel, keys [[0,0],[3,0],[7,1],[16,0]]}
         ]
        }]
      }]
    }
    Image_IconDuplicated {
     id=#4
     possessed_class=Image
     tracks [
      {
       type=MovieSceneFloatTrack
       display_name=Render Opacity
       sections [{
         type=MovieSceneFloatSection
         end_frame=30
         start_frame=0
         channels [None {type=MovieSceneScriptingFloatChannel, keys [[0,0],[3,0],[8,1],[25,1],[30,0]]}]
        }]
      }
      {
       type=MovieScene2DTransformTrack
       sections [{
         type=MovieScene2DTransformSection
         end_frame=12
         start_frame=0
         channels [
          Scale.X {type=MovieSceneScriptingFloatChannel, keys [[0,2],[3,5],[8,0.8],[12,0.9]]}
          Scale.Y {type=MovieSceneScriptingFloatChannel, keys [[0,2],[3,5],[8,0.8],[12,0.9]]}
         ]
        }]
      }
     ]
    }
   ]}
 }
 Anim_Outro {
  binding_count=1
  length_seconds=0.4
  object=&self:Anim_Outro
  sequence {bindings [PrimaryWidthSB {
     id=#5
     possessed_class=SizeBox
     tracks [{
       type=MovieSceneFloatTrack
       display_name=Render Opacity
       sections [{type=MovieSceneFloatSection, end_frame=8, start_frame=0, channels [None {type=MovieSceneScriptingFloatChannel, keys [[0,1],[8,0]]}]}]
      }]
    }]}
 }
]
graphs [
 EventGraph {nodes [
   n1 = SetBackplateBrushFromSoftMaterial!() -> (OutputDelegate, then, SoftMat:Material Interface Soft Object Reference) {is=title K2Node_CustomEvent}
   > n2 = Async Load Asset(Asset=@n1.SoftMat)
   SetBrushFromMaterial(
    execute=@n2.Completed
    self:Border Object Reference=Get AnimBoundCB() -> (AnimBoundCB:Common Border Object Reference)
    Material=Cast To MaterialInstance!(Object=@n2) -> (AsMaterial Instance:Material Instance Object Reference, bSuccess:bool=~) {is=title K2Node_DynamicCast}
   )
   n3 = SetOutlineSize!() -> (OutputDelegate, then, Outline Size:int) {is=title K2Node_CustomEvent}
   n4 = Get Text_Detail() -> (Text_Detail:Common Rich Text Block Object Reference)
   n5 = Get!(self:Rich Text Block Object Reference=@n4) -> (
    DefaultTextStyleOverride_Font_FontObject:Object Reference
    DefaultTextStyleOverride_Font_FontMaterial:Object Reference
    DefaultTextStyleOverride_Font_OutlineSettings:Font Outline Settings Structure
    DefaultTextStyleOverride_Font_TypefaceFontName:Name
    DefaultTextStyleOverride_Font_Size:float=~
    DefaultTextStyleOverride_Font_LetterSpacing:int
    DefaultTextStyleOverride_Font_SkewAmount:float=~
    DefaultTextStyleOverride_Font_bForceMonospaced:bool=~
    DefaultTextStyleOverride_Font_bMaterialIsStencil:bool=~
    DefaultTextStyleOverride_Font_MonospacedWidth:float=~
    DefaultTextStyleOverride_ColorAndOpacity:Slate Color Structure
    DefaultTextStyleOverride_ShadowOffset:Vector 2D Structure
    DefaultTextStyleOverride_ShadowColorAndOpacity:Linear Color Structure
    DefaultTextStyleOverride_HighlightColor:Slate Color Structure
    DefaultTextStyleOverride_HighlightShape:Slate Brush Structure
    DefaultTextStyleOverride_StrikeBrush:Slate Brush Structure
    DefaultTextStyleOverride_UnderlineBrush:Slate Brush Structure
    DefaultTextStyleOverride_TransformPolicy:ETextTransformPolicy Enum
    DefaultTextStyleOverride_OverflowPolicy:ETextOverflowPolicy Enum
   )
   > SetDefaultFont!(
    execute
    self:Rich Text Block Object Reference=@n4
    InFontInfo_FontObject:Object Reference=@n5
    InFontInfo_FontMaterial:Object Reference=@n5.DefaultTextStyleOverride_Font_FontMaterial
    InFontInfo_OutlineSettings:Font Outline Settings Structure=Make Font Outline Settings!(
     OutlineSize:int=@n3.Outline Size
     bMiteredCorners:bool=False
     bSeparateFillAlpha:bool=False
     bApplyOutlineToDropShadows:bool=False
     OutlineMaterial:Object Reference
     OutlineColor:Linear Color Structure=(R=0,G=0,B=0,A=1)
    ) -> (FontOutlineSettings:Font Outline Settings Structure) {is=title K2Node_MakeStruct}
    InFontInfo_TypefaceFontName:Name=@n5.DefaultTextStyleOverride_Font_TypefaceFontName
    InFontInfo_Size:float=@n5.DefaultTextStyleOverride_Font_Size
    InFontInfo_LetterSpacing:int=@n5.DefaultTextStyleOverride_Font_LetterSpacing
    InFontInfo_SkewAmount:float=@n5.DefaultTextStyleOverride_Font_SkewAmount
    InFontInfo_bForceMonospaced:bool=@n5.DefaultTextStyleOverride_Font_bForceMonospaced
    InFontInfo_bMaterialIsStencil:bool=@n5.DefaultTextStyleOverride_Font_bMaterialIsStencil
    InFontInfo_MonospacedWidth:float=@n5.DefaultTextStyleOverride_Font_MonospacedWidth
   ) -> (then)
   Event PreConstruct()
   > SetMobileStyle() {is=title K2Node_CallFunction}
  ]}
 PlayIntroAnimation {nodes [
   PlayIntroAnimation() {is=title K2Node_FunctionEntry}
   > PlayAnimation(InAnimation=Get Anim_Intro() -> (Anim_Intro:Widget Animation Object Reference))
   n1 = Get MessageSoundCue() -> (MessageSoundCue:Sound Base Object Reference)
   > Is Valid(InputObject=@n1)
   > PlaySound2D(Sound=@n1)
  ]}
 PlayOutroAnimation {nodes [
   PlayOutroAnimation() {is=title K2Node_FunctionEntry}
   > PlayAnimation(InAnimation=Get Anim_Outro() -> (Anim_Outro:Widget Animation Object Reference))
  ]}
 SetBackplateColors {nodes [
   n1 = SetBackplateColors!() -> (then, OutlineColor:Linear Color Structure, GlowColor:Linear Color Structure) {is=title K2Node_FunctionEntry}
   > n2 = GetDynamicMaterial(self:Border Object Reference=Get AnimBoundCB())
   > SetVectorParameterValue(self=@n2, ParameterName=Rarity01, Value=@n1)
   > SetVectorParameterValue(self=@n2, ParameterName=Rarity03, Value=@n1.GlowColor)
  ]}
 SetBrushFromMaterial {nodes [
   n1 = SetBrushFromMaterial!() -> (then, Material:Material Interface Object Reference) {is=title K2Node_FunctionEntry}
   > SetBrushFromMaterial(
    self:Image Object Reference=[Get Image_IconDuplicated() -> (Image_IconDuplicated:Image Object Reference), Get Image_Icon() -> (Image_Icon:Image Object Reference)]
    Material=@n1
   ) {is=title K2Node_CallFunction}
  ]}
 SetBrushFromSoftMaterial {nodes [
   n1 = SetBrushFromSoftMaterial!() -> (then, SoftMaterial:Material Interface Soft Object Reference) {is=title K2Node_FunctionEntry}
   > SetBrushFromSoftMaterial!(
    execute
    self:Image Object Reference=[Get Image_IconDuplicated(), Get Image_Icon()]
    SoftMaterial:Material Interface Soft Object Reference=@n1
   ) -> (then) {is=title K2Node_CallFunction}
  ]}
 SetBrushFromTexture {nodes [
   n1 = SetBrushFromTexture!() -> (then, Texture:Texture 2D Object Reference) {is=title K2Node_FunctionEntry}
   > SetBrushFromTexture!(
    execute
    self:Image Object Reference=[Get Image_IconDuplicated(), Get Image_Icon()]
    Texture:Texture 2D Object Reference=@n1
    bMatchSize:bool=false
   ) -> (then) {is=title K2Node_CallFunction}
  ]}
 SetDetails {nodes [
   n1 = SetDetails!() -> (then, InText:Text) {is=title K2Node_FunctionEntry}
   n2 = Get Text_Detail()
   > SetText(self=@n2, InText=@n1)
   > SetVisibility(self=@n2, InVisibility=Select(Option 0=SelfHitTestInvisible, Option 1=Collapsed, Index=TextIsEmpty(InText=@n1)))
  ]}
 SetHeadling {nodes [
   n1 = SetHeadling!() -> (then, InText:Text) {is=title K2Node_FunctionEntry}
   > "SetText (Text)"(self=Get Text_Headline() -> (Text_Headline:Common Text Object Reference), InText=@n1)
  ]}
 SetMobileStyle {nodes [
   SetMobileStyle() {is=title K2Node_FunctionEntry}
   n1 = GetOwningPlayer()
   > Branch(Condition=AND Boolean(
     A=IsUsingCompactUI()
     B=IsUsingInputType!(InputType:ECommonInputType Enum=Touch, PlayerController:Player Controller Object Reference=@n1) -> (ReturnValue:bool)
    ))
   > Set WrapTextAt!(execute, WrapTextAt:float=275, self:Text Layout Widget Object Reference=Get Text_Detail()) -> (then, Output_Get:float)
   > SetWrapTextWidth!(execute, self:Common Text Object Reference=Get Text_Headline(), InWrapTextAt:int=275) -> (then)
  ]}
]
variables [MessageSoundCue {type=SoundBase, default=null}]
widgets {
 info {inheritedWidgetCount=0, parentClass=&CommonUserWidget, rootWidgetClass=&SizeBox, widgetCount=13}
 widgets [
  PrimaryWidthSB {
   widgetClassPath=&SizeBox
   widget=PrimaryWidthSB
   properties {bOverride_MaxDesiredWidth=true, heightOverride=158.2614, maxDesiredWidth=672, widthOverride=99.6082}
  }
  Overlay_2 {widgetClassPath=&Overlay, parent=PrimaryWidthSB, widget=Overlay_2, slot {class=SizeBoxSlot}}
  AnimBoundCB {
   widgetClassPath=&CommonBorder
   bIsVariable=true
   parent=Overlay_2
   widget=AnimBoundCB
   properties {
    clipping=ClipToBounds
    horizontalAlignment=HAlign_Center
    verticalAlignment=VAlign_Center
    background {
     bIsDynamicallyLoaded=false
     drawAs=Image
     imageType=FullColor
     mirroring=NoMirror
     resourceName=None
     resourceObject=/Game/Athena/UI/Materials/M_UI_NotificationBackplate
     tiling=NoTile
     imageSize=<32 32>
     margin=<ltrb 0 0 0 0>
     outlineSettings {
      bUseBrushTransparency=false
      roundingType=HalfHeightRadius
      width=0
      color {colorUseRule=UseColor_Specified, specifiedColor=<rgba 0 0 0 0>}
      cornerRadii=<0 0 0 0>
     }
     tintColor {colorUseRule=UseColor_Specified, specifiedColor=<rgba 1 1 1 1>}
     uVRegion {bIsValid=false, max=<0 0>, min=<0 0>}
    }
    padding=<ltrb 50 50 50 50>
   }
   slot {class=OverlaySlot, properties {verticalAlignment=VAlign_Fill}}
  }
  HorizontalBox_1 {
   widgetClassPath=&HorizontalBox
   parent=AnimBoundCB
   widget=HorizontalBox_1
   slot {class=BorderSlot, properties {horizontalAlignment=HAlign_Center, verticalAlignment=VAlign_Center, padding=<ltrb 50 50 50 50>}}
  }
  Overlay_0 {
   widgetClassPath=&Overlay
   parent=HorizontalBox_1
   widget=Overlay_0
   slot {class=HorizontalBoxSlot, properties {horizontalAlignment=HAlign_Left, verticalAlignment=VAlign_Center}}
  }
  Image_Icon {
   widgetClassPath=&Image
   bIsVariable=true
   parent=Overlay_0
   widget=Image_Icon
   properties {brush {
     bIsDynamicallyLoaded=false
     drawAs=Image
     imageType=NoImage
     mirroring=NoMirror
     resourceName=None
     resourceObject=null
     tiling=NoTile
     imageSize=<64 64>
     margin=<ltrb 0 0 0 0>
     outlineSettings {
      bUseBrushTransparency=false
      roundingType=HalfHeightRadius
      width=0
      color {colorUseRule=UseColor_Specified, specifiedColor=<rgba 0 0 0 0>}
      cornerRadii=<0 0 0 0>
     }
     tintColor {colorUseRule=UseColor_Specified, specifiedColor=<rgba 1 1 1 1>}
     uVRegion {bIsValid=false, max=<0 0>, min=<0 0>}
    }}
   slot {class=OverlaySlot, properties {horizontalAlignment=HAlign_Center, verticalAlignment=VAlign_Center}}
  }
  NamedSlot_Icon {
   widgetClassPath=&NamedSlot
   bIsVariable=true
   parent=Overlay_0
   widget=NamedSlot_Icon
   slot {class=OverlaySlot, properties {horizontalAlignment=HAlign_Fill, verticalAlignment=VAlign_Fill}}
  }
  VerticalBox_Contents {
   widgetClassPath=&VerticalBox
   bIsVariable=true
   parent=HorizontalBox_1
   widget=VerticalBox_Contents
   slot {class=HorizontalBoxSlot, properties {verticalAlignment=VAlign_Center, padding=<ltrb 16 0 0 0>, size {sizeRule=Fill, value=1}}}
  }
  HorizontalBox_2 {
   widgetClassPath=&HorizontalBox
   parent=VerticalBox_Contents
   widget=HorizontalBox_2
   slot {class=VerticalBoxSlot, properties {padding=<ltrb 0 0 0 7>, size {sizeRule=Fill, value=1}}}
  }
  Text_Headline {
   widgetClassPath=&CommonTextBlock
   bIsVariable=true
   parent=HorizontalBox_2
   widget=Text_Headline
   properties {
    mobileFontSizeMultiplier=1.1
    text="{Headline}"
    textTransformPolicy=ToUpper
    wrapTextAt=550
    font {
     bForceMonospaced=false
     bMaterialIsStencil=false
     fontMaterial=null
     fontObject=&Burbank
     letterSpacing=0
     monospacedWidth=1
     size=16.5
     skewAmount=0
     typefaceFontName=BurbankRegular-Black
     outlineSettings {
      bApplyOutlineToDropShadows=false
      bMiteredCorners=false
      bSeparateFillAlpha=false
      outlineMaterial=null
      outlineSize=0
      outlineColor=<rgba 0 0 0 1>
     }
    }
    margin=<ltrb 0 -8 0 -8>
   }
   slot {class=HorizontalBoxSlot}
  }
  NamedSlot_RightOfHeader {
   widgetClassPath=&NamedSlot
   bIsVariable=true
   parent=HorizontalBox_2
   widget=NamedSlot_RightOfHeader
   slot {class=HorizontalBoxSlot}
  }
  Text_Detail {
   widgetClassPath=&CommonRichTextBlock
   bIsVariable=true
   parent=VerticalBox_Contents
   widget=Text_Detail
   properties {
    bOverrideDefaultStyle=true
    mobileTextBlockScale=1.2
    text="{DETAILS <text color=\"FF0000FF\" size=\"18\">60</>}"
    wrapTextAt=400
    defaultTextStyleOverride {
     overflowPolicy=Clip
     transformPolicy=ToUpper
     colorAndOpacity {colorUseRule=UseColor_Specified, specifiedColor=<rgba 1 1 1 1>}
     font {
      bForceMonospaced=false
      bMaterialIsStencil=false
      fontMaterial=null
      fontObject=&BurbankSmall
      letterSpacing=0
      monospacedWidth=1
      size=11.25
      skewAmount=0
      typefaceFontName=Bold
      outlineSettings {
       bApplyOutlineToDropShadows=false
       bMiteredCorners=false
       bSeparateFillAlpha=false
       outlineMaterial=null
       outlineSize=0
       outlineColor=<rgba 0 0 0 1>
      }
     }
     highlightColor {colorUseRule=UseColor_Specified, specifiedColor=<rgba 0 0 0 1>}
     highlightShape {
      bIsDynamicallyLoaded=false
      drawAs=Image
      imageType=NoImage
      mirroring=NoMirror
      resourceName=None
      resourceObject=null
      tiling=NoTile
      imageSize=<32 32>
      margin=<ltrb 0 0 0 0>
      outlineSettings {
       bUseBrushTransparency=false
       roundingType=HalfHeightRadius
       width=0
       color {colorUseRule=UseColor_Specified, specifiedColor=<rgba 0 0 0 0>}
       cornerRadii=<0 0 0 0>
      }
      tintColor {colorUseRule=UseColor_Specified, specifiedColor=<rgba 1 1 1 1>}
      uVRegion {bIsValid=false, max=<0 0>, min=<0 0>}
     }
     selectedBackgroundColor {colorUseRule=UseColor_Foreground, specifiedColor=<rgba 1 0 1 1>}
     shadowColorAndOpacity=<rgba 0 0 0 1>
     shadowOffset=<0 0>
     strikeBrush {
      bIsDynamicallyLoaded=false
      drawAs=Image
      imageType=NoImage
      mirroring=NoMirror
      resourceName=None
      resourceObject=null
      tiling=NoTile
      imageSize=<32 32>
      margin=<ltrb 0 0 0 0>
      outlineSettings {
       bUseBrushTransparency=false
       roundingType=HalfHeightRadius
       width=0
       color {colorUseRule=UseColor_Specified, specifiedColor=<rgba 0 0 0 0>}
       cornerRadii=<0 0 0 0>
      }
      tintColor {colorUseRule=UseColor_Specified, specifiedColor=<rgba 1 1 1 1>}
      uVRegion {bIsValid=false, max=<0 0>, min=<0 0>}
     }
     underlineBrush {
      bIsDynamicallyLoaded=false
      drawAs=Image
      imageType=NoImage
      mirroring=NoMirror
      resourceName=None
      resourceObject=null
      tiling=NoTile
      imageSize=<32 32>
      margin=<ltrb 0 0 0 0>
      outlineSettings {
       bUseBrushTransparency=false
       roundingType=HalfHeightRadius
       width=0
       color {colorUseRule=UseColor_Specified, specifiedColor=<rgba 0 0 0 0>}
       cornerRadii=<0 0 0 0>
      }
      tintColor {colorUseRule=UseColor_Specified, specifiedColor=<rgba 1 1 1 1>}
      uVRegion {bIsValid=false, max=<0 0>, min=<0 0>}
     }
    }
   }
   slot {class=VerticalBoxSlot}
  }
  Image_IconDuplicated {
   widgetClassPath=&Image
   bIsVariable=true
   parent=Overlay_2
   widget=Image_IconDuplicated
   properties {
    renderOpacity=0
    brush {
     bIsDynamicallyLoaded=false
     drawAs=Image
     imageType=NoImage
     mirroring=NoMirror
     resourceName=None
     resourceObject=null
     tiling=NoTile
     imageSize=<90 90>
     margin=<ltrb 0 0 0 0>
     outlineSettings {
      bUseBrushTransparency=false
      roundingType=HalfHeightRadius
      width=0
      color {colorUseRule=UseColor_Specified, specifiedColor=<rgba 0 0 0 0>}
      cornerRadii=<0 0 0 0>
     }
     tintColor {colorUseRule=UseColor_Specified, specifiedColor=<rgba 1 1 1 1>}
     uVRegion {bIsValid=false, max=<0 0>, min=<0 0>}
    }
   }
   slot {class=OverlaySlot, properties {horizontalAlignment=HAlign_Center, verticalAlignment=VAlign_Center}}
  }
 ]
}
