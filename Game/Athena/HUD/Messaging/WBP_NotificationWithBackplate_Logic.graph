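# HUD notification logic widget that feeds notification data into its root WBP_NotificationWithBackplate_Visuals widget.
# On Construct it finds BP_NotificationManagerComponent on the owning player, subscribes to new-data events, and applies the currently displaying notification; Destruct unsubscribes.
# Valid data sets headline, details, icon material and, if provided, backplate material through soft references.
# Invalid data hides the widget, and the valid path never restores visibility.
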WidgetBlueprint /Game/Athena/HUD/Messaging/WBP_NotificationWithBackplate_Logic
graphs [
 EventGraph {nodes [
   Event Construct()
   > Parent: Construct()
   > n1 = Set NotificationManager(NotificationManager:BP Notification Manager Component Object Reference=GetComponentByClass!(
     self:Actor Object Reference=GetOwningPlayer()
     ComponentClass:Actor Component Class Reference=&/BP_NotificationManagerComponent.BP_NotificationManagerComponent_C
    ) -> (ReturnValue:BP Notification Manager Component Object Reference)) -> (Output_Get:BP Notification Manager Component Object Reference)
   > Is Valid(InputObject=@n1)
   > Bind Event to On New Notification Data(self:BP Notification Manager Component Object Reference=@n1, Delegate=Create Event()) {is=title K2Node_AddDelegate}
   > OnNewNotificationData!(
    execute
    self
    NewNotificationData:BP Notification Data Type Object Reference=GetNotificationCurrentlyDisplaying!(self:BP Notification Manager Component Object Reference=@n1) -> (NotificationData:BP Notification Data Type Object Reference)
   ) -> (then)
   Event Destruct()
   > Parent: Destruct()
   n2 = Get NotificationManager() -> (NotificationManager:BP Notification Manager Component Object Reference)
   > Is Valid(InputObject=@n2)
   > Unbind Event from On New Notification Data(self:BP Notification Manager Component Object Reference=@n2, Delegate=Create Event()) {is=title K2Node_RemoveDelegate}
  ]}
 OnNewNotificationData {nodes [
   n1 = OnNewNotificationData!() -> (then, NewNotificationData:BP Notification Data Type Object Reference) {is=title K2Node_FunctionEntry}
   > n2 = Is Valid(InputObject=@n1)
   SetVisibility(execute=@n2.Is Not Valid, InVisibility=Hidden)
   n3 = Get WBP_NotificationWithBackplate_Visuals() -> (WBP_NotificationWithBackplate_Visuals:WBP Notification With Backplate Visuals Object Reference)
   SetHeadling!(
    execute=@n2.Is Valid
    self:WBP Notification With Backplate Visuals Object Reference=@n3
    InText:Text=Get Headline!(self:BP Notification Data Type Object Reference=@n1) -> (Headline:Text)
   ) -> (then)
   > SetDetails!(
    execute
    self:WBP Notification With Backplate Visuals Object Reference=@n3
    InText:Text=Get Detail!(self:BP Notification Data Type Object Reference=@n1) -> (Detail:Text)
   ) -> (then)
   > SetBrushFromSoftMaterial!(
    execute
    self:WBP Notification With Backplate Visuals Object Reference=@n3
    SoftMaterial:Material Interface Soft Object Reference=Get Icon!(self:BP Notification Data Type Object Reference=@n1) -> (Icon:Material Interface Soft Object Reference)
   ) -> (then)
   n4 = Get BackplateBrush!(self:BP Notification Data Type Object Reference=@n1) -> (BackplateBrush:Material Interface Soft Object Reference)
   > Branch(Condition=IsValidSoftObjectReference(SoftObjectReference=@n4))
   > SetBackplateBrushFromSoftMaterial!(execute, self:WBP Notification With Backplate Visuals Object Reference=@n3, SoftMat:Material Interface Soft Object Reference=@n4) -> (then)
  ]}
]
variables [NotificationManager {type=BP_NotificationManagerComponent_C, default=null}]
widgets {
 info {
  inheritedWidgetCount=0
  parentClass=&UserWidget
  rootWidgetClass=&/WBP_NotificationWithBackplate_Visuals.WBP_NotificationWithBackplate_Visuals_C
  widgetCount=1
 }
 widgets [WBP_NotificationWithBackplate_Visuals {
   widgetClassPath=&/WBP_NotificationWithBackplate_Visuals
   bIsVariable=true
   widget=WBP_NotificationWithBackplate_Visuals
  }]
}
